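# Post-copy animation blueprint for the M_MED_Lexa body mesh that adds procedural motion to the character's choker.
# It copies the pose from the attached parent mesh, then constrains the collar bone to a weighted blend of head, neck, clavicle and upper spine bones.
# The ring base follows the collar, and an Anim Dynamics node simulates the ring in actor space with tight angular limits and heavy damping.
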AnimBlueprint /BRCosmetics/Characters/Player/Male/Medium/Bodies/M_MED_Lexa/Meshes/M_MED_Lexa_AnimBP
target_skeleton=&/M_MED_Lexa_Skeleton
graphs [
 AnimGraph {nodes [
   n1 = "Constraint - dyn_choker_collar (PPPPPP)"!(
    ComponentPose=Local To Component(LocalPose=Copy Pose From Mesh() {bUseAttachedParent=true})
    Alpha=1
    ConstraintWeights_0=0.0667
    ConstraintWeights_1:float=0.0254
    ConstraintWeights_2:float=0.6601
    ConstraintWeights_3:float=0.0541
    ConstraintWeights_4:float=0.053
    ConstraintWeights_5:float=0.1406
   ) -> (Pose) {
    alphaBoolBlend {blendOption=Linear}
    boneToModify {boneName=dyn_choker_collar}
    constraintSetup [
     {targetBone {boneName=head}}
     {targetBone {boneName=clavicle_l}}
     {targetBone {boneName=spine_05}}
     {targetBone {boneName=clavicle_r}}
     {targetBone {boneName=neck_02}}
     {targetBone {boneName=neck_01}}
    ]
    constraintWeights [1,1,1,1,1,1]
   }
   n2 = "Constraint - dyn_choker_ring_base (P)"!(ComponentPose=@n1, Alpha=1, ConstraintWeights_0=1, ConstraintWeights_1:float=0) -> (Pose) {
    alphaBoolBlend {blendOption=Linear}
    boneToModify {boneName=dyn_choker_ring_base}
    constraintSetup [{targetBone {boneName=dyn_choker_collar}}]
    constraintWeights [1,0]
   }
   n3 = Anim Dynamics - Bone: dyn_choker_ring_parent(ComponentPose=@n2, Alpha=1, ExternalForce="0,0,0") {
    angularDampingOverride=0.999
    bGravityOverrideInSimSpace=true
    bOverrideAngularDamping=true
    bOverrideLinearDamping=true
    linearDampingOverride=0.999
    simulationSpace=Actor
    alphaBoolBlend {blendOption=Linear}
    alphaScaleBiasClamp {bInterpResult=true, interpSpeedDecreasing=4, interpSpeedIncreasing=6}
    boundBone {boneName=dyn_choker_ring_parent}
    componentAppliedLinearAccClamp=<6000 6000 6000>
    componentLinearAccScale=<0.25 0.25 0.25>
    componentLinearVelScale=<0.25 0.25 0.25>
    physicsBodyDefinitions [{
      sphereCollisionRadius=0
      boundBone {boneName=dyn_choker_ring_parent}
      boxExtents=<8 8 3>
      constraintSetup {angularLimitsMax=<5 60 0>, angularLimitsMin=<-5 -8 0>, angularTarget=<0 0 0>}
      localJointOffset=<2 0 0>
     }]
    relativeSpaceBone {boneName=spine_05}
   }
   Output Pose(Result=Component To Local(ComponentPose=@n3)) {layerGroup=None}
  ]}
 EventGraph {nodes [Event BlueprintUpdateAnimation()]}
]
variables []
