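# Physics asset for the M_MED_Lexa character's hair parts, defining bodies and joint limits for dynamic hair bones around the head.
# Capsules cover the head and hair strands, with small spheres on the offset and bang base bones.
# Most strands constrain directly to the head with tight swing and twist limits.
# Bangs form short left and right chains from their bases; those bases and the offset have no constraint to the head.
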PhysicsAsset /BRCosmetics/Characters/Player/Male/Medium/Bodies/M_MED_Lexa/Meshes/Parts/M_MED_Lexa_Physics
bodies {
 dyn_hair_01 [{shapeType=Capsule, length=1.2596, radius=0.9932, center=<1.0056 0 -0.0001>, rotation=<rot 90 0 0>}]
 dyn_hair_010 [{shapeType=Capsule, length=1.4172, radius=1.1651, center=<1.345 -0.0001 0>, rotation=<rot 90 0 0>}]
 dyn_hair_011 [{shapeType=Capsule, length=1.6123, radius=1.4142, center=<1.8117 0.0001 0>, rotation=<rot 90 0 0>}]
 dyn_hair_012 [{shapeType=Capsule, length=1.3306, radius=0.8407, center=<1.105 0 0>, rotation=<rot 90 0 0>}]
 dyn_hair_013 [{shapeType=Capsule, length=1.2194, radius=0.9617, center=<1.2016 0 0.0002>, rotation=<rot 90 0 0>}]
 dyn_hair_014 [{shapeType=Capsule, length=1.1783, radius=1.2702, center=<1.8489 -0.0001 0>, rotation=<rot 90 0 0>}]
 dyn_hair_02 [{shapeType=Capsule, length=2.7322, radius=1.4876, center=<1.8617 0 0>, rotation=<rot 90 0 0>}]
 dyn_hair_03 [{shapeType=Capsule, length=1.867, radius=1.2552, center=<1.5365 0 0>, rotation=<rot 90 0 0>}]
 dyn_hair_04 [{shapeType=Capsule, length=1.4158, radius=1.5233, center=<1.672 0 0>, rotation=<rot 90 0 0>}]
 dyn_hair_05 [{shapeType=Capsule, length=1.8789, radius=0.9107, center=<1.2101 -0.0001 0>, rotation=<rot 90 0 0>}]
 dyn_hair_06 [{shapeType=Capsule, length=1.4824, radius=1.3214, center=<1.8025 0 -0.0001>, rotation=<rot 90 0 0>}]
 dyn_hair_07 [{shapeType=Capsule, length=1.9303, radius=1.5447, center=<1.2719 0 0>, rotation=<rot 90 0 0>}]
 dyn_hair_08 [{shapeType=Capsule, length=1.4979, radius=0.982, center=<1.2706 0 -0.0003>, rotation=<rot 90 0 0>}]
 dyn_hair_09 [{shapeType=Capsule, length=1.0221, radius=0.8041, center=<1.0765 0 -0.0001>, rotation=<rot 90 0 0>}]
 dyn_hair_bang_01_l [{shapeType=Capsule, length=2.0431, radius=1.0095, center=<1.8834 0.0001 0.1992>, rotation=<rot 82.1798 179.9998 179.9999>}]
 dyn_hair_bang_01_r [{shapeType=Capsule, length=1.66, radius=0.6073, center=<1.3568 0 0>, rotation=<rot 90 0 0>}]
 dyn_hair_bang_02_l [{shapeType=Capsule, length=1.5365, radius=0.574, center=<1.4736 0.0001 0>, rotation=<rot 90 0 0>}]
 dyn_hair_bang_base_l [{shapeType=Sphere, radius=0.5}]
 dyn_hair_bang_base_r [{shapeType=Sphere, radius=0.5}]
 dyn_hair_bang_parent_l [{shapeType=Capsule, length=3.0147, radius=1.2067, center=<2.3961 0 0.3196>, rotation=<rot 81.189 -179.9998 -179.9998>}]
 dyn_hair_bang_parent_r [{shapeType=Capsule, length=2.1744, radius=1.2536, center=<1.7123 0 0>, rotation=<rot 90 0 0>}]
 dyn_hair_offset [{shapeType=Sphere, radius=0.5}]
 dyn_hair_top_back_01 [{shapeType=Capsule, length=3.1228, radius=1.0196, center=<1.7851 0 0>, rotation=<rot 90 0 0>}]
 dyn_hair_top_front_01 [{shapeType=Capsule, length=1.2732, radius=1.1956, center=<1.3136 0 0.0001>, rotation=<rot 90 0 0>}]
 dyn_hair_top_mid_01 [{shapeType=Capsule, length=1.9554, radius=1.0699, center=<1.2034 0 0>, rotation=<rot 90 0 0>}]
 head [{shapeType=Capsule, length=7.0823, radius=7.7313, center=<5.635 2.1445 0.6486>, rotation=<rot 11.5238 100.8236 -89.9999>}]
}
constraints [
 {bone1Name=dyn_hair_bang_parent_l, bone2Name=dyn_hair_bang_base_l, swing1LimitDegrees=5, swing2LimitDegrees=5, twistLimitDegrees=3}
 {bone1Name=dyn_hair_bang_parent_r, bone2Name=dyn_hair_bang_base_r, swing1LimitDegrees=5, swing2LimitDegrees=20, twistLimitDegrees=3}
 {bone1Name=dyn_hair_bang_01_r, bone2Name=dyn_hair_bang_parent_r, swing1LimitDegrees=10, swing2LimitDegrees=5, twistLimitDegrees=3}
 {bone1Name=dyn_hair_bang_01_l, bone2Name=dyn_hair_bang_parent_l, swing1LimitDegrees=10, swing2LimitDegrees=10, twistLimitDegrees=3}
 {bone1Name=dyn_hair_bang_02_l, bone2Name=dyn_hair_bang_01_l, swing1LimitDegrees=20, swing2LimitDegrees=20, twistLimitDegrees=3}
 {bone1Name=dyn_hair_01, bone2Name=dyn_hair_offset, swing1LimitDegrees=20, swing2LimitDegrees=20, twistLimitDegrees=3}
 {bone1Name=dyn_hair_02, bone2Name=head, swing1LimitDegrees=10, swing2LimitDegrees=10, twistLimitDegrees=3}
 {bone1Name=dyn_hair_03, bone2Name=head, swing1LimitDegrees=6, swing2LimitDegrees=6, twistLimitDegrees=3}
 {bone1Name=dyn_hair_04, bone2Name=head, swing1LimitDegrees=20, swing2LimitDegrees=20, twistLimitDegrees=3}
 {bone1Name=dyn_hair_05, bone2Name=head, swing1LimitDegrees=15, swing2LimitDegrees=15, twistLimitDegrees=3}
 {bone1Name=dyn_hair_06, bone2Name=head, swing1LimitDegrees=10, swing2LimitDegrees=10, twistLimitDegrees=3}
 {bone1Name=dyn_hair_08, bone2Name=head, swing1LimitDegrees=15, swing2LimitDegrees=15, twistLimitDegrees=3}
 {bone1Name=dyn_hair_09, bone2Name=head, swing1LimitDegrees=15, swing2LimitDegrees=15, twistLimitDegrees=3}
 {bone1Name=dyn_hair_07, bone2Name=head, swing1LimitDegrees=20, swing2LimitDegrees=20, twistLimitDegrees=3}
 {bone1Name=dyn_hair_013, bone2Name=head, swing1LimitDegrees=6, swing2LimitDegrees=6, twistLimitDegrees=3}
 {bone1Name=dyn_hair_010, bone2Name=head, swing1LimitDegrees=12, swing2LimitDegrees=12, twistLimitDegrees=3}
 {bone1Name=dyn_hair_011, bone2Name=head, swing1LimitDegrees=20, swing2LimitDegrees=20, twistLimitDegrees=3}
 {bone1Name=dyn_hair_012, bone2Name=head, swing1LimitDegrees=15, swing2LimitDegrees=15, twistLimitDegrees=3}
 {bone1Name=dyn_hair_014, bone2Name=head, swing1LimitDegrees=15, swing2LimitDegrees=15, twistLimitDegrees=3}
 {bone1Name=dyn_hair_top_front_01, bone2Name=head, swing1LimitDegrees=20, swing2LimitDegrees=20, twistLimitDegrees=3}
 {bone1Name=dyn_hair_top_mid_01, bone2Name=head, swing1LimitDegrees=20, swing2LimitDegrees=20, twistLimitDegrees=3}
 {bone1Name=dyn_hair_top_back_01, bone2Name=head, swing1LimitDegrees=10, swing2LimitDegrees=10, twistLimitDegrees=1}
]
